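# Editor utility widget for level designers that lists overlapping World Partition actors.
# A Python utility returns boxes keyed by pipe-delimited ActorDesc strings, which further Python parses back into structs for list items.
# Selecting an item shows its descriptors and draws a debug box, Shift-click pins its actors to load them, and right-click focuses the viewport.
# A save game keeps results and the "Always Highlight Selection" option between sessions.
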EditorUtilityWidgetBlueprint /Game/Blueprints/Blutilities/ForLDs/WorldPartitionTools/WP_OverlappingActors/WP_OverlappingActors_EUW
graphs [
 DrawDebugHighlight {nodes [
   n1 = DrawDebugHighlight!() -> (then, min:Vector=~, max:Vector=~) {is=title K2Node_FunctionEntry}
   > DrawDebugBox!(
    execute
    WorldContextObject:Object Reference
    Center:Vector=vector / vector(A:Vector=vector + vector(A=@n1, B=@n1.max), B=2) -> (ReturnValue:Vector)
    Extent:Vector=vector / vector(A=vector - vector(A=@n1.max, B=@n1), B=2)
    LineColor:Linear Color Structure=(R=0.4219,G=0.0511,B=0.0301,A=1)
    Rotation:Rotator
    Duration:float=3
    Thickness:float=30
    DepthPriority:EDrawDebugSceneDepthPriorityGroup Enum=World
   ) -> (then)
  ]}
 EventGraph {nodes [
   "On Clicked (TA_Generic01_Button)"!() -> (OutputDelegate, then, widget:User Widget Object Reference, event:Pointer Event Structure) {is=title K2Node_ComponentBoundEvent}
   > FindDescs()
   n1 = "On Item Left Clicked (TA_Generic01_List)"!() -> (OutputDelegate, then, isShiftDown:bool, isCtrlDown:bool, data:Object Reference, item:User Widget Object Reference) {is=title K2Node_ComponentBoundEvent}
   > Branch(Condition=@n1.isShiftDown)
   > ItemLoadAreaWithVolume!(execute, self, Object:Object Reference=@n1.data) -> (then)
   n2 = "On Selection Changed (TA_Generic01_Checkbox)"!() -> (OutputDelegate, then, selected:bool, widget:User Widget Object Reference) {is=title K2Node_ComponentBoundEvent}
   > SaveGame_Save(self:TA Base Editor Widget Object Reference)
   > n3 = Branch(Condition=@n2.selected)
   > n4 = Branch(Condition=Is Valid Timer Handle!(Handle:Timer Handle Structure=Get highlightTimer() -> (highlightTimer:Timer Handle Structure)) -> (ReturnValue:bool))
   n5 = Set Timer by Event(execute=@n4.else, Delegate=Create Event(), Time=0.25, bLooping=false)
   > Set highlightTimer(highlightTimer:Timer Handle Structure=@n5) -> (Output_Get:Timer Handle Structure)
   Clear and Invalidate Timer by Handle!(execute=@n3.else, WorldContextObject:Object Reference, Handle=Get highlightTimer()) -> (then)
   n6 = "On Item Right Clicked (TA_Generic01_List)"!() -> (OutputDelegate, then, isShiftDown:bool, isCtrlDown:bool, data:Object Reference, item:User Widget Object Reference) {is=title K2Node_ComponentBoundEvent}
   > ItemFocus!(execute, self, Object:Object Reference=@n6.data) -> (then)
   n7 = "On Item Selection Changed (TA_Generic01_List)"!() -> (
    OutputDelegate
    then
    selectedData:Array of Object References
    selectedWidgets:Array of User Widget Object References
    selectedIndices:Array of Integers
   ) {is=title K2Node_ComponentBoundEvent}
   > Branch(Condition=Is Valid Index!(TargetArray:Array of Object References=@n7.selectedData, IndexToTest:int) -> (ReturnValue:bool))
   > HandleItemSelection!(execute, self, Object:Object Reference="Get (a copy)"(Array:Array of Object References=@n7.selectedData) -> (Output:Object Reference)) -> (then)
  ]}
 FindDescs {nodes [
   FindDescs() {is=title K2Node_FunctionEntry}
   > clearItems(self:TA Generic 01 List Object Reference=Get TA_Generic01_List() -> (TA_Generic01_List:TA Generic 01 List Object Reference))
   > n1 = Execute Python Script!(
    execute
    PythonScript="import unreal\r\nimport LevelDesign.GeneralLevelUtils as utils\r\nimport importlib\r\n\r\nimportlib.reload(utils)\r\n\r\nboxes = utils.getOverlappingActorDescs()"
   ) -> (then, ReturnValue, boxes:Map of Strings to Box Structures) {outputs [boxes]}
   > PopulateListFromMapping!(execute, self, TargetMap:Map of Strings to Box Structures=@n1.boxes) -> (then)
   > SaveGame_Save()
  ]}
 HandleItemSelection {nodes [
   n1 = HandleItemSelection!() -> (then, Object:Object Reference) {is=title K2Node_FunctionEntry}
   > n2 = Cast To WP_OverlappingActors_ListItemData!(execute, Object=@n1) -> (then, CastFailed, AsWP Overlapping Actors List Item Data:WP Overlapping Actors List Item Data Object Reference) {is=title K2Node_DynamicCast}
   > n3 = Sequence()
   > Set ActorDescriptionData(ActorDescriptionData:Array of Actor Desc Structures=Get ActorDescriptionData!(self:WP Overlapping Actors List Item Data Object Reference=@n2) -> (ActorDescriptionData:Array of Actor Desc Structures)) -> (Output_Get:Array of Actor Desc Structures)
   > SetActiveWidgetIndex(self=Get switcherDetails() -> (switcherDetails:Widget Switcher Object Reference), Index=1)
   > SetObject(self=Get DetailsView_115() -> (DetailsView_115:Details View Object Reference), NewObject=Self-Reference())
   > Set Text(
    Text="To Text (String)"(InString=Get displayString!(self:TA Generic List Item Data Object Reference=@n2) -> (displayString:String))
    self=Get txtDetailsHeader() -> (txtDetailsHeader:Common Text Object Reference)
   )
   n4 = Get vectors!(self:TA Generic List Item Data Object Reference=@n2) -> (vectors:Array of Vectors)
   DrawDebugHighlight!(
    execute=@n3.then_1
    self
    min:Vector="Get (a ref)"(Array:Array of Vectors=@n4, Dimension 1=0) -> (Output:"Vector (by ref)")
    max:Vector="Get (a ref)"(Array=@n4, Dimension 1=1)
   ) -> (then) {is=title K2Node_CallFunction}
   SetObject(execute=@n2.CastFailed, self=Get DetailsView_115())
   > SetActiveWidgetIndex(self=Get switcherDetails(), Index=0)
   n5 = For Each Loop(Array:Array of Actor Desc Structures=Get ActorDescriptionData()) -> (Array Element:Actor Desc Structure)
   > clearItems(self=Get listDescs() -> (listDescs:TA Generic 01 List Object Reference))
   > n6 = Construct Object from Class(Class=&/WP_OverlappingActors_ListItemData.WP_OverlappingActors_ListItemData_C) -> (ReturnValue:WP Overlapping Actors List Item Data Object Reference)
   > n7 = Set ActorDescriptionData!(
    execute
    ActorDescriptionData=Make Array("[0]":Actor Desc Structure=@n5) -> (Array:Array of Actor Desc Structures)
    self:WP Overlapping Actors List Item Data Object Reference=@n6
   ) -> (then, Output_Get)
   n8 = "Get (a copy)"!(Array:Array of Actor Desc Structures=@n7, Dimension 1) -> (
    Output_Guid:Guid Structure
    Output_NativeClass:Object Class Reference
    Output_Class:Soft Object Path Structure
    Output_Name:Name
    Output_Label:Name
    Output_Bounds:Box Structure
    Output_RuntimeGrid:Name
    Output_bIsSpatiallyLoaded:bool=~
    Output_bActorIsEditorOnly:bool=~
    Output_ActorPackage:Name
    Output_ActorPath:Name
    Output_FolderPath:Name
    Output_DataLayerAssets:Array of Soft Object Path Structures
   )
   n9 = "To String (Name)"(InName=@n8.Output_Label)
   > Set displayString!(execute, displayString:String=@n9, self:TA Generic List Item Data Object Reference=@n6) -> (then, Output_Get:String)
   n10 = Break Box!(Box:"Box Structure (by ref)"=@n8.Output_Bounds) -> (Min:Vector=~, Max:Vector=~, IsValid:bool=~) {is=title K2Node_BreakStruct}
   > Set vectors!(
    execute
    vectors:Array of Vectors=Make Array!("[0]":Vector=@n10, "[1]":Vector=@n10.Max) -> (Array:Array of Vectors)
    self:TA Generic List Item Data Object Reference=@n6
   ) -> (then, Output_Get:Array of Vectors)
   > addItem!(execute, self:TA Generic 01 List Object Reference=Get listDescs(), dataObject:Object Reference=@n6) -> (then, newItem:User Widget Object Reference, newData:Object Reference)
   > PrintString(InString=Append(A="Adding ", B=@n9))
  ]}
 ItemFocus {nodes [
   n1 = ItemFocus!() -> (then, Object:Object Reference) {is=title K2Node_FunctionEntry}
   > n2 = Cast To WP_OverlappingActors_ListItemData(Object=@n1)
   n3 = Get vectors(self=@n2)
   n4 = "Get (a ref)"(Array=@n3, Dimension 1=1)
   n5 = "Get (a ref)"(Array=@n3, Dimension 1=0)
   n6 = vector + vector(A=@n5, B=@n4)
   PrintString(InString=Append!(A="Focusing at ", B="To String (Vector)"!(InVec:Vector=@n6) -> (ReturnValue:String), C:String, D:String) -> (ReturnValue))
   > PrintString(InString=Append!(A="           min:", B="To String (Vector)"(InVec=@n5)) -> (ReturnValue))
   > n7 = PrintString(InString=Append(A="           max:", B="To String (Vector)"(InVec=@n4)))
   Focus Bounds in Viewport!(" ":Exec=[@n7.then, @n2.then], center:Vector=vector / vector(A=@n6, B=2), bounds:Vector=vector - vector(A=@n4, B=@n5)) -> ("  ":Exec) {is=title K2Node_MacroInstance}
   > DrawDebugHighlight(min=@n5, max=@n4)
  ]}
 ItemHighlight {nodes [
   ItemHighlight() {is=title K2Node_FunctionEntry}
   > n1 = Sequence()
   > n2 = getSelectedData!(execute, self:TA Generic 01 List Object Reference=Get TA_Generic01_List()) -> (then, selectedData:Array of Object References)
   > Branch(Condition=Is Valid Index(TargetArray=@n2))
   > n3 = Cast To TA_GenericListItemData!(execute, Object="Get (a copy)"!(Array:Array of Object References=@n2, Dimension 1) -> (Output:Object Reference)) -> (then, CastFailed, AsTA Generic List Item Data:TA Generic List Item Data Object Reference) {is=title K2Node_DynamicCast}
   n4 = Get vectors(self=@n3)
   > DrawDebugHighlight(min="Get (a ref)"(Array=@n4, Dimension 1=0), max="Get (a ref)"(Array=@n4, Dimension 1=1))
   n5 = Branch(
    execute=@n1.then_1
    Condition=getSelected!(self:TA Base Editor Button Widget Object Reference=Get TA_Generic01_Checkbox() -> (TA_Generic01_Checkbox:TA Generic 01 Checkbox Object Reference)) -> (isSelected:bool)
   )
   > n6 = Set Timer by Event(Delegate=Create Event(), Time=0.25, bLooping=false)
   > Set highlightTimer(highlightTimer=@n6)
   Clear and Invalidate Timer by Handle(execute=@n5.else, Handle=Get highlightTimer())
  ]}
 ItemLoadArea {nodes [
   n1 = ItemLoadArea!() -> (then, Object:Object Reference) {is=title K2Node_FunctionEntry}
   > n2 = Cast To TA_GenericListItemData(Object=@n1)
   n3 = Get vectors(self=@n2)
   > n4 = GetIntersectingActorDescs!(
    execute
    InBox:"Box Structure (by ref)"=MakeBox!(Min:Vector="Get (a ref)"(Array=@n3, Dimension 1=0), Max:Vector="Get (a ref)"(Array=@n3, Dimension 1=1)) -> (ReturnValue:Box Structure)
   ) -> (then, OutActorDescs:Array of Actor Desc Structures, ReturnValue:bool)
   > n5 = For Each Loop(Array=@n4)
   PinActors!(execute=@n5.Completed, InActorsToPin:Array of Guid Structures=Get l_guidsToPin() -> (l_guidsToPin:Array of Guid Structures)) -> (then)
   Add(
    execute=@n5.LoopBody
    TargetArray:Array of Guid Structures=Get l_guidsToPin()
    NewItem:"Guid Structure (by ref)"=Break Actor Desc!(ActorDesc:"Actor Desc Structure (by ref)"=@n5) -> (
     Guid:Guid Structure
     NativeClass:Object Class Reference
     Class:Soft Object Path Structure
     Name:Name
     Label:Name
     Bounds:Box Structure
     RuntimeGrid:Name
     bIsSpatiallyLoaded:bool=~
     bActorIsEditorOnly:bool=~
     ActorPackage:Name
     ActorPath:Name
     FolderPath:Name
     DataLayerAssets:Array of Soft Object Path Structures
    ) {is=title K2Node_BreakStruct}
   ) {is=title K2Node_CallArrayFunction}
  ]}
 ItemLoadAreaWithVolume {nodes [
   n1 = ItemLoadAreaWithVolume!() -> (then, Object:Object Reference) {is=title K2Node_FunctionEntry}
   > n2 = Cast To WP_OverlappingActors_ListItemData(Object=@n1)
   > n3 = For Each Loop(Array=Get ActorDescriptionData(self=@n2))
   n4 = Get l_guids() -> (l_guids:Array of Guid Structures)
   PinActors(execute=@n3.Completed, InActorsToPin=@n4)
   Add(execute=@n3.LoopBody, TargetArray=@n4, NewItem=Break Actor Desc(ActorDesc=@n3) -> (bIsSpatiallyLoaded=~, bActorIsEditorOnly=~))
  ]}
 PopulateListFromMapping {nodes [
   n1 = PopulateListFromMapping!() -> (then, TargetMap:Map of Strings to Box Structures) {is=title K2Node_FunctionEntry}
   > Set StringBoxMap(StringBoxMap:Map of Strings to Box Structures=@n1) -> (Output_Get:Map of Strings to Box Structures)
   > clearItems(self=Get TA_Generic01_List())
   > n2 = Keys!(execute, TargetMap:Map of Strings to Box Structures=@n1) -> (then, Keys:Array of Strings)
   > n3 = For Each Loop(Array:Array of Strings=@n2) -> (Array Element:String)
   > n4 = Values!(execute, TargetMap:Map of Strings to Box Structures=@n1) -> (then, Values:Array of Box Structures)
   > Set l_box(l_box:Box Structure="Get (a ref)"(Array:Array of Box Structures=@n4, Dimension 1=@n3.Array Index) -> (Output:"Box Structure (by ref)")) -> (Output_Get:Box Structure)
   > Set l_array(l_array:Array of Strings=ParseIntoArray(SourceString=@n3, Delimiter=|)) -> (Output_Get:Array of Strings)
   > Set l_tooltip(l_tooltip:String) -> (Output_Get:String)
   > n5 = Sequence!(execute) -> (then_0, then_1, then_2:Exec)
   > Clear(TargetArray:Array of Actor Desc Structures=Get l_descs() -> (l_descs:Array of Actor Desc Structures))
   > n6 = For Each Loop(Array=Get l_array() -> (l_array:Array of Strings))
   > n7 = Execute Python Script!(execute, PythonScript="import unreal\r\nout_struct = unreal.ActorDesc()\r\nout_struct.import_text(string_data)", string_data:String=@n6) -> (then, ReturnValue, out_struct:"Actor Desc Structure (by ref)") {inputs [string_data], outputs [out_struct]}
   > Add(TargetArray:Array of Actor Desc Structures=Get l_descs(), NewItem:"Actor Desc Structure (by ref)"=@n7.out_struct)
   > n8 = Branch(Condition="Not Equal (Integer)"(A=@n6.Array Index))
   > n9 = Set l_tooltip(l_tooltip=Append(A=Get l_tooltip() -> (l_tooltip:String), B="\r\n"))
   n10 = Break Actor Desc(ActorDesc=@n7.out_struct) -> (bIsSpatiallyLoaded=~, bActorIsEditorOnly=~)
   Set l_tooltip(
    execute=[@n8.else, @n9.then]
    l_tooltip=Append!(A=Get l_tooltip(), B="   -", C:String="To String (Name)"(InName=@n10.Label)) -> (ReturnValue)
   )
   n11 = Construct Object from Class(
    execute=@n5.then_2
    Class=&WP_OverlappingActors_ListItemData_C
   )
   > n12 = Set ActorDescriptionData(ActorDescriptionData=Get l_descs(), self=@n11)
   > Set displayString(
    displayString="To String (Name)"(InName="Get (a copy)"!(Array:Array of Actor Desc Structures=@n12, Dimension 1) -> (
      Output_Guid:Guid Structure
      Output_NativeClass:Object Class Reference
      Output_Class:Soft Object Path Structure
      Output_Name:Name
      Output_Label:Name
      Output_Bounds:Box Structure
      Output_RuntimeGrid:Name
      Output_bIsSpatiallyLoaded:bool=~
      Output_bActorIsEditorOnly:bool=~
      Output_ActorPackage:Name
      Output_ActorPath:Name
      Output_FolderPath:Name
      Output_DataLayerAssets:Array of Soft Object Path Structures
     ).Output_Label)
    self=@n11
   )
   n13 = Get!() -> (l_box_Min:Vector=~, l_box_Max:Vector=~, l_box_IsValid:bool=~)
   > Set vectors(vectors=Make Array("[0]"=@n13, "[1]"=@n13.l_box_Max), self=@n11)
   > n14 = addItem(self=Get TA_Generic01_List(), dataObject=@n11)
   n15 = Get l_tooltip()
   > SetToolTipText!(
    execute
    self:Widget Object Reference=@n14
    InToolTipText:"Text (by ref)"="To Text (String)"(InString=Append!(A="Click to Show Details, Shift Click to Load, Right Click to Focus\r\n\r\nActors:\r\n", B=@n15) -> (ReturnValue))
   ) -> (then)
  ]}
 SaveGame_PostLoad {nodes [
   n1 = SaveGame_PostLoad!() -> (then, saveGameExisted:bool, saveGame_Object:Save Game Object Reference) {is=title K2Node_FunctionEntry}
   > n2 = Cast To WP_OVerlappingActors_SaveGame!(execute, Object=@n1.saveGame_Object) -> (then, CastFailed, AsWP OVerlapping Actors Save Game:WP OVerlapping Actors Save Game Object Reference) {is=title K2Node_DynamicCast}
   > Set StringBoxMap(StringBoxMap=Get StringBoxMap!(self:WP OVerlapping Actors Save Game Object Reference=@n2) -> (StringBoxMap:Map of Strings to Box Structures))
   > n3 = Sequence!(execute) -> (then_0, then_1)
   n4 = Get StringBoxMap!() -> (StringBoxMap)
   > Branch(Condition="integer > integer"(A=Length!(TargetMap:Map of Strings to Box Structures=@n4) -> (ReturnValue:int=0) {is=title K2Node_CallFunction}))
   > PopulateListFromMapping(TargetMap=Get StringBoxMap()) {is=title K2Node_CallFunction}
   setSelected!(
    execute=@n3.then_1
    self:TA Base Editor Button Widget Object Reference=Get TA_Generic01_Checkbox()
    selected:bool=Get AlwaysHighlightSelection!(self:WP OVerlapping Actors Save Game Object Reference=@n2) -> (AlwaysHighlightSelection:bool)
    throwEvent:bool=true
   ) -> (then)
   Return Node!(execute, " ":bool=false)
  ]}
 SaveGame_PrepareToSave {nodes [
   n1 = SaveGame_PrepareToSave!() -> (then, SaveGameObject:Save Game Object Reference) {is=title K2Node_FunctionEntry}
   > n2 = Cast To WP_OVerlappingActors_SaveGame(Object=@n1)
   > Set StringBoxMap!(execute, StringBoxMap=Get StringBoxMap(), self:WP OVerlapping Actors Save Game Object Reference=@n2) -> (then, Output_Get)
   > Set AlwaysHighlightSelection!(
    execute
    AlwaysHighlightSelection:bool=getSelected(self=Get TA_Generic01_Checkbox())
    self:WP OVerlapping Actors Save Game Object Reference=@n2
   ) -> (then, Output_Get:bool)
   Return Node(" "=false)
  ]}
]
variables [
 ActorDescriptionData {type="array<ActorDesc>", default []}
 Box {type="array<Vector>"}
 Boxes {type="array<Box>", default []}
 highlightTimer {type=TimerHandle}
 StringBoxMap {type="map<Box>", default {}}
]
widgets {
 info {
  parentClass=/Game/Blueprints/Blutilities/ForTechArt/Widgets/TA_Base_EditorWidget.TA_Base_EditorWidget_C
  rootWidgetClass=&VerticalBox
  widgetCount=14
 }
 widgets [
  VerticalBox_1 {widgetClassPath=&VerticalBox, widget=VerticalBox_1}
  HorizontalBox_0 {widgetClassPath=&HorizontalBox, parent=VerticalBox_1, widget=HorizontalBox_0, slot {class=VerticalBoxSlot}}
  TA_Generic01_Button {
   widgetClassPath=&TA_Generic01_Button
   parent=HorizontalBox_0
   widget=TA_Generic01_Button
   properties {label=Find Overlapping Actors}
   slot {class=HorizontalBoxSlot, properties {verticalAlignment=VAlign_Center, size {sizeRule=Fill, value=1}}}
  }
  TA_Generic01_Checkbox {
   widgetClassPath=/Game/Blueprints/Blutilities/ForTechArt/Widgets/Buttons/TA_Generic01_Checkbox
   bIsVariable=true
   parent=HorizontalBox_0
   widget=TA_Generic01_Checkbox
   properties {label=Always Highlight Selection}
   slot {class=HorizontalBoxSlot, properties {verticalAlignment=VAlign_Center, padding=<ltrb 5 0 0 0>}}
  }
  Image_139 {
   widgetClassPath=&Image
   bIsVariable=true
   parent=VerticalBox_1
   widget=Image_139
   properties {
    brush {
     bIsDynamicallyLoaded=false
     drawAs=Image
     imageType=NoImage
     mirroring=NoMirror
     resourceName=None
     resourceObject=null
     tiling=NoTile
     imageSize=<32 4>
     margin=<ltrb 0 0 0 0>
     outlineSettings {
      bUseBrushTransparency=false
      roundingType=HalfHeightRadius
      width=0
      color {colorUseRule=UseColor_Specified, specifiedColor=<rgba 0 0 0 0>}
      cornerRadii=<0 0 0 0>
     }
     tintColor {colorUseRule=UseColor_Specified, specifiedColor=<rgba 1 1 1 1>}
     uVRegion {bIsValid=false, max=<0 0>, min=<0 0>}
    }
    colorAndOpacity=<rgba 1 1 1 0.25>
   }
   slot {class=VerticalBoxSlot}
  }
  Border_0 {
   widgetClassPath=&Border
   parent=VerticalBox_1
   widget=Border_0
   properties {brushColor=<rgba 1 1 1 0.15>, padding=<ltrb 5 5 5 5>}
   slot {class=VerticalBoxSlot, properties {padding=<ltrb 5 5 5 5>, size {sizeRule=Fill, value=1}}}
  }
  HorizontalBox_1 {
   widgetClassPath=&HorizontalBox
   parent=Border_0
   widget=HorizontalBox_1
   slot {class=BorderSlot, properties {padding=<ltrb 5 5 5 5>}}
  }
  TA_Generic01_List {
   widgetClassPath=/Game/Blueprints/Blutilities/ForTechArt/Widgets/List/TA_Generic01_List
   bIsVariable=true
   parent=HorizontalBox_1
   widget=TA_Generic01_List
   properties {allowMultiSelect=false}
   slot {class=HorizontalBoxSlot, properties {size {sizeRule=Fill, value=1}}}
  }
  Border_1 {
   widgetClassPath=&Border
   parent=HorizontalBox_1
   widget=Border_1
   properties {brushColor=<rgba 0 0 0 0.563>}
   slot {class=HorizontalBoxSlot, properties {size {sizeRule=Fill, value=1}}}
  }
  switcherDetails {widgetClassPath=&WidgetSwitcher, bIsVariable=true, parent=Border_1, widget=switcherDetails, slot {class=BorderSlot}}
  CommonTextBlock_73 {
   widgetClassPath=&CommonTextBlock
   parent=switcherDetails
   widget=CommonTextBlock_73
   properties {
    text="<- Select a list item to display details"
    colorAndOpacity {colorUseRule=UseColor_Specified, specifiedColor=<rgba 0.3125 0.3125 0.3125 1>}
    font {
     bForceMonospaced=false
     bMaterialIsStencil=false
     fontMaterial=null
     fontObject=&NotoSans
     letterSpacing=0
     monospacedWidth=1
     size=16
     skewAmount=0
     typefaceFontName=NotoSans-Regular
     outlineSettings {
      bApplyOutlineToDropShadows=false
      bMiteredCorners=false
      bSeparateFillAlpha=false
      outlineMaterial=null
      outlineSize=0
      outlineColor=<rgba 0 0 0 1>
     }
    }
    shadowOffset=<0 0>
   }
   slot {class=WidgetSwitcherSlot, properties {horizontalAlignment=HAlign_Center, verticalAlignment=VAlign_Top, padding=<ltrb 15 15 15 15>}}
  }
  vboxDetails {widgetClassPath=&VerticalBox, bIsVariable=true, parent=switcherDetails, widget=vboxDetails, slot {class=WidgetSwitcherSlot}}
  txtDetailsHeader {
   widgetClassPath=&CommonTextBlock
   bIsVariable=true
   parent=vboxDetails
   widget=txtDetailsHeader
   properties {
    colorAndOpacity {colorUseRule=UseColor_Specified, specifiedColor=<rgba 0.9115 0.9115 0.9115 1>}
    font {
     bForceMonospaced=false
     bMaterialIsStencil=false
     fontMaterial=null
     fontObject=&NotoSans
     letterSpacing=0
     monospacedWidth=1
     size=16
     skewAmount=0
     typefaceFontName=NotoSans-Regular
     outlineSettings {
      bApplyOutlineToDropShadows=false
      bMiteredCorners=false
      bSeparateFillAlpha=false
      outlineMaterial=null
      outlineSize=0
      outlineColor=<rgba 0 0 0 1>
     }
    }
    shadowOffset=<0 0>
   }
   slot {class=VerticalBoxSlot, properties {horizontalAlignment=HAlign_Center, verticalAlignment=VAlign_Top}}
  }
  DetailsView_115 {
   widgetClassPath=&DetailsView
   parent=vboxDetails
   widget=DetailsView_115
   properties {propertiesToShow [ActorDescriptionData]}
   slot {class=VerticalBoxSlot, properties {size {sizeRule=Fill, value=1}}}
  }
 ]
}
